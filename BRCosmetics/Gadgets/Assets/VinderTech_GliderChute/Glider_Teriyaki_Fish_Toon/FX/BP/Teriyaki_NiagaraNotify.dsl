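; Niagara anim notify for the Teriyaki Fish Toon glider that spawns an effect and configures it after spawning.
; Received_Notify calls the parent notify first, then fetches the spawned effect and checks it is valid.
; On a valid effect it sets render custom depth, an "Index" int parameter and translucent sort priority from the notify's CustomDepth, Index and Sort variables.
; It returns the parent's result either way.
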(fn Received_Notify
  (params (MeshComp :type Skeletal%20Mesh%20Component%20Object%20Reference) (Animation :type Anim%20Sequence%20Base%20Object%20Reference) (EventReference :type Anim%20Notify%20Event%20Reference%20Structure%20%28by%20ref%29))
  (bind _returnvalue (|CallParentFunction MeshComp Animation))
  (bind _returnvalue_2 (AnimNotify|GetSpawnedEffect))
  (Utilities|IsValid _returnvalue_2
    (:Is%20Valid
      (Rendering|SetRenderCustomDepth _returnvalue_2 (Variables|Default|GetCustomDepth))
      (Effects|Components|ParticleSystem|SetIntParameter _returnvalue_2 "Index" (Variables|Default|GetIndex))
      (Rendering|SetTranslucentSortPriority _returnvalue_2 (Variables|Default|GetSort)))
    (:Is%20Not%20Valid
      _))
  (return _returnvalue))
