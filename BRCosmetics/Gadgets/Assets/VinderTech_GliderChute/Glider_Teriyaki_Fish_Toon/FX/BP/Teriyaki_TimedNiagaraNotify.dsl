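; Timed Niagara anim notify state for the Teriyaki Fish Toon glider that configures the effect it spawns on the skeletal mesh.
; On notify begin it calls the parent implementation, then retrieves the spawned effect from the mesh component.
; If that effect is valid, it enables custom depth rendering, passes the BoneIndex variable into the effect's "Index" int parameter, and applies the Sort variable as translucent sort priority.
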(fn Received_NotifyBegin
  (params (MeshComp :type Skeletal%20Mesh%20Component%20Object%20Reference) (Animation :type Anim%20Sequence%20Base%20Object%20Reference) (TotalDuration :type Float%20%28single-precision%29) (EventReference :type Anim%20Notify%20Event%20Reference%20Structure%20%28by%20ref%29))
  (bind _returnvalue (|CallParentFunction MeshComp Animation TotalDuration))
  (bind _returnvalue_2 (AnimNotify|GetSpawnedEffect MeshComp))
  (Utilities|IsValid _returnvalue_2
    (:Is%20Valid
      (Rendering|SetRenderCustomDepth _returnvalue_2 true)
      (Effects|Components|ParticleSystem|SetIntParameter _returnvalue_2 "Index" (Variables|Default|GetBoneIndex))
      (Rendering|SetTranslucentSortPriority _returnvalue_2 (Variables|Default|GetSort)))
    (:Is%20Not%20Valid
      _))
  (return _returnvalue))
