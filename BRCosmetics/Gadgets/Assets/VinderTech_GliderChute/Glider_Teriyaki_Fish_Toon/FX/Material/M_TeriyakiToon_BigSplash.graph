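# Unlit, two-sided masked toon water splash material for Niagara mesh particles on the Teriyaki Fish Toon glider.
# Dynamic parameters drive vertex offset along local Z and the tangent-normal cross product, shaped by V-based curves.
# Three offset splash alpha samples multiply with constant or particle color, leaving a dark outline.
# High-contrast panning noise against a V gradient dissolves the splash alpha; its UV scale and movement reuse the U parameters for V.
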Material /BRCosmetics/Gadgets/Assets/VinderTech_GliderChute/Glider_Teriyaki_Fish_Toon/FX/Material/M_TeriyakiToon_BigSplash
graph {
 properties {
  bUsedWithNiagaraMeshParticles=true
  blendMode=BLEND_Masked
  pixelDepthOffsetMode=PDOM_Legacy
  shadingModel=MSM_Unlit
  twoSided=true
  usedShadingModels=Unlit
 }
 nodes [
  n1 = DynamicParameter() {parameterIndex=1, defaultValue=<rgba 100 -100 1.2 0.7>, paramNames [HeightWPO, ExpandWPO, HeightWPOCurve, ExpandWPOCurve]}
  n2 = Normalize(VectorInput=Transform(None=Constant3Vector() {constant=<rgba 0 0 1 0>}) {transformSourceType=TRANSFORMSOURCE_Local})
  n3 = MaterialFunctionCall(Float2=TextureCoordinate()) {materialFunction=&BreakOutFloat2Components}
  n4 = Power(Base=OneMinus(None=@n3.G), Exp=@n1.HeightWPOCurve)
  n5 = Add(A=@n3.G)
  n6 = Power(Base=Sine(None=Multiply(A=@n5) {constB=0.25}), Exp=@n1.ExpandWPOCurve)
  n7 = CrossProduct(A=VertexTangentWS(), B=VertexNormalWS())
  Add(A=Multiply(A=@n1.HeightWPO, B=Multiply(A=@n2, B=@n4)), B=Multiply(A=@n1.ExpandWPO, B=Multiply(A=@n6, B=@n7)))
  Constant()
  n8 = Constant()
  n9 = DynamicParameter() {defaultValue=<rgba 0 0.01 2 0>, paramNames [Dissolve,Outline,VPower,VOffset]}
  n10 = Multiply(A=@n9.Outline) {constB=-1}
  n11 = DynamicParameter() {parameterIndex=3, defaultValue=<rgba 1 0 0 0>, paramNames [UScale, URandom, TimeRandom, ""]}
  n12 = MaterialFunctionCall(Float2=TextureCoordinate()) {materialFunction=&BreakOutFloat2Components}
  n13 = Add(
   A=MaterialFunctionCall(X=Multiply(A=@n11.UScale, B=@n12.R), Y=Power(Base=@n12.G, Exp=@n9.VPower)) {materialFunction=&MakeFloat2}.Result
   B=MaterialFunctionCall(X=@n11.URandom, Y=@n9.VOffset) {materialFunction=&MakeFloat2}.Result
  )
  n14 = Add(A=@n13, B=AppendVector(A=@n8, B=@n10))
  n15 = VectorParameter() {group=Color, parameterName=Color, defaultValue=<rgba 0.6 0.6 0.6 1>}
  n16 = ParticleColor()
  n17 = ScalarParameter() {group=Color, parameterName=UseParticleColor}
  n18 = Add(A=@n13, B=AppendVector(A=@n10, B=@n8))
  n19 = Add(A=@n13, B=AppendVector(A=@n9.Outline, B=@n8))
  n20 = Multiply(
   A=Multiply(
    A=TextureSample(UVs=@n14) {texture=/BRCosmetics/Gadgets/Assets/VinderTech_GliderChute/Glider_Teriyaki_Fish_Toon/FX/Texture/T_water_splash}.A
    B=LinearInterpolate(A=@n15.RGB, B=@n16.RGB, Alpha=@n17)
   )
   B=Multiply(
    A=TextureSample(UVs=@n18) {texture=&T_water_splash}.A
    B=TextureSample(UVs=@n19) {texture=&T_water_splash}.A
   )
  )
  MaterialFunctionCall(Emissive=ShadingPathSwitch(Default=@n20, Forward=@n20, Mobile=@n20)) {materialFunction=&MF_FX_EmissiveAdjustments}
  n21 = DynamicParameter() {parameterIndex=2, defaultValue=<rgba 3 0.5 0 0>, paramNames [DissolveUScale, DissolveVScale, DissolveUMove, DissolveVMove]}
  n22 = Multiply(A=TextureCoordinate(), B=AppendVector(A=@n21.DissolveUScale, B=@n21.DissolveUScale))
  n23 = Multiply(A=AppendVector(A=@n21.DissolveUMove, B=@n21.DissolveUMove), B=Add(A=@n11.TimeRandom, B=Time()))
  n24 = MaterialFunctionCall(Float2=TextureCoordinate()) {materialFunction=&BreakOutFloat2Components}
  n25 = LinearInterpolate(
   A=Constant() {r=-1.43}
   B=Constant() {r=0.07}
   Alpha=DynamicParameter() {defaultValue=<rgba 0 0.01 2 0>, paramNames [Dissolve,Outline,VPower,VOffset]}.Dissolve
  )
  n26 = Subtract(
   A=TextureSample(UVs=Add(A=@n22, B=@n23)) {samplerType=SAMPLERTYPE_Grayscale, texture=&LowResBlurredNoise}.R
   B=Clamp(None=Add(A=@n24.G, B=@n25)) {minDefault=-1}
  )
  n27 = Constant() {r=15}
  Multiply(B=Multiply(
    A=MaterialFunctionCall(In=@n26, Contrast=@n27) {materialFunction=&CheapContrast}.Result
    B=TextureSample(UVs=@n13) {texture=&T_water_splash}.A
   ))
 ]
}
