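# Unlit, masked, two-sided material for Niagara sprites, named as the toon drip effect of the Teriyaki Fish Toon glider.
# It samples the T_ToonDrip texture and scales its RGB down by a constant multiplier.
# The dimmed color feeds the Emissive input of the shared MF_FX_EmissiveAdjustments material function.
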Material /BRCosmetics/Gadgets/Assets/VinderTech_GliderChute/Glider_Teriyaki_Fish_Toon/FX/Material/M_TeriyakiToon_Drip
graph {
 properties {
  bUsedWithNiagaraSprites=true
  blendMode=BLEND_Masked
  pixelDepthOffsetMode=PDOM_Legacy
  shadingModel=MSM_Unlit
  twoSided=true
  usedShadingModels=Unlit
 }
 nodes [MaterialFunctionCall(Emissive=Multiply(A=TextureSample() {texture=/BRCosmetics/Gadgets/Assets/VinderTech_GliderChute/Glider_Teriyaki_Fish_Toon/FX/Texture/T_ToonDrip}.RGB) {constB=0.6}) {materialFunction=&MF_FX_EmissiveAdjustments}]
}
