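# Unlit translucent, two-sided material for the toon ink-flow effect on the Teriyaki Fish glider, flagged for Niagara mesh particles.
# A base shape texture gets its UVs warped by two panning noise layers and scaled by Appear.
# Vertex offset math combines bending, sine and noise waves, and tail expansion.
# Dynamic parameters drive time, direction response and strength; custom-depth outline color logic sits mostly under a "NOOOTUse" parameter group.
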Material /BRCosmetics/Gadgets/Assets/VinderTech_GliderChute/Glider_Teriyaki_Fish_Toon/FX/Material/M_TeriyakiToon_InkFlow
graph {
 properties {
  allowTranslucentCustomDepthWrites=true
  bUsedWithNiagaraMeshParticles=true
  blendMode=BLEND_Translucent
  opacityMaskClipValue=0.005
  pixelDepthOffsetMode=PDOM_Legacy
  shadingModel=MSM_Unlit
  twoSided=true
  usedShadingModels=Unlit
 }
 nodes [
  n1 = MaterialFunctionCall(Float2=MaterialFunctionCall(Float2=TextureCoordinate()) {materialFunction=&BreakOutFloat2Components}.R) {materialFunction=&BreakOutFloat2Components}
  n2 = OneMinus(None=Sine(None=Multiply(A=@n1.R) {constB=0.5}))
  n3 = VertexNormalWS()
  n4 = DynamicParameter() {paramNames [RightDot,AreaPower,Str,Appear]}
  n5 = Power(
   Base=MaterialFunctionCall(Float2=TextureCoordinate()) {materialFunction=&BreakOutFloat2Components}.G
   Exp=@n4.AreaPower
  )
  n6 = Constant3Vector() {constant=<rgba 1 0 0 0>}
  n7 = Multiply(A=Multiply(A=Multiply(A=@n4.RightDot) {constB=-1}, B=Transform(None=@n6)), B=@n4.Str)
  n8 = MaterialFunctionCall(Float2=TextureCoordinate()) {materialFunction=&BreakOutFloat2Components}
  n9 = Multiply(A=@n8.G, B=ScalarParameter() {defaultValue=1, parameterName=SineWPOFrequency})
  n10 = DynamicParameter() {parameterIndex=1, paramNames [Time, ForwardDot, TempFIxWPOSpeed, TempFIxWPOStr]}
  n11 = DynamicParameter() {parameterIndex=1, paramNames [Time, ForwardDot, TempFIxWPOSpeed, TempFIxWPOStr]}
  n12 = Multiply(A=@n10.TempFIxWPOSpeed, B=Multiply(A=@n11.Time, B=ScalarParameter() {defaultValue=-1, parameterName=SineWPOSpeed}))
  n13 = Multiply(A=Sine(None=Add(A=@n9, B=@n12)), B=VertexNormalWS())
  n14 = Multiply(A=@n10.TempFIxWPOStr, B=ScalarParameter() {defaultValue=20, parameterName=SineWPOStrength})
  n15 = VectorParameter() {parameterName=NoiseWPOUV, defaultValue=<rgba 1 1 0 -0.7>}
  n16 = Multiply(A=TextureCoordinate(), B=AppendVector(A=@n15.R, B=@n15.R))
  n17 = Multiply(A=@n11.Time, B=AppendVector(A=@n15.B, B=@n15.B))
  n18 = Multiply(
   A=TextureSample(UVs=Add(A=@n16, B=@n17)) {texture=/Game/Effects/Fort_Effects/Textures/Power_Source/T_Power_Source_Sphere_Texture}.RGB
   B=VertexNormalWS()
  )
  n19 = ScalarParameter() {defaultValue=40, parameterName=NoiseWPOStr}
  n20 = Constant3Vector() {constant=<rgba 1 0 0 0>}
  n21 = ScalarParameter() {defaultValue=200, parameterName=TailExpandStr}
  n22 = Multiply(A=@n21) {constB=-1}
  n23 = Add(
   A=Add(A=Multiply(A=@n13, B=@n14), B=Multiply(A=@n18, B=@n19))
   B=Multiply(A=Transform(None=@n20), B=LinearInterpolate(A=@n22, B=@n21, Alpha=@n8.R))
  )
  n24 = Power(Base=@n8.G, Exp=ScalarParameter() {defaultValue=1, parameterName=SineWPOGradientPower})
  Add(
   A=Multiply(A=Multiply(A=@n2, B=@n3), B=ScalarParameter() {defaultValue=-100, parameterName=UDirectBendStr})
   B=Add(A=Multiply(A=@n5, B=@n7), B=Multiply(A=@n23, B=@n24))
  )
  Constant()
  n25 = DepthFade(FadeDistance=ScalarParameter() {defaultValue=5, group=NOOOTUse, parameterName=FD})
  n26 = Saturate(None=Multiply(A=OneMinus(None=@n25)) {constB=100})
  n27 = VectorParameter() {group=NOOOTUse, parameterName=InnerColor, defaultValue=<rgba 0 0 0 1>}
  n28 = Multiply(A=@n26, B=@n27.RGB) {constB=0.7}
  n29 = VectorParameter() {group=NOOOTUse, parameterName=FC, defaultValue=<rgba 0 0 0 1>}
  n30 = VectorParameter() {group=NOOOTUse, parameterName=OutColor, defaultValue=<rgba 0 0 0 1>}
  n31 = LinearInterpolate(A=@n27.RGB, B=@n30.RGB, Alpha=@n26)
  n32 = ComponentMask(None=SceneTexture() {sceneTextureId=PPI_CustomDepth}.Color) {r=true}
  n33 = ScalarParameter() {defaultValue=1.5, group=NOOOTUse, parameterName=CullDepth}
  n34 = PixelDepth()
  n35 = Constant() {r=1}
  n36 = ScreenPosition()
  n37 = Subtract(A=Divide(A=ViewSize(), B=Constant() {r=320}), B=Constant() {r=2}) {constB=0}
  n38 = ViewSize()
  n39 = Divide(
   A=Divide(A=ComponentMask(None=@n38) {r=true}, B=ComponentMask(None=@n38) {g=true})
   B=MaterialFunctionCall() {materialFunction=/Engine/Functions/Engine_MaterialFunctions02/Utility/FOV}."FOV "
  )
  n40 = ObjectPositionWS()
  n41 = ParticlePositionWS()
  n42 = Multiply(A=Distance(A=StaticSwitchParameter(True=@n40, False=@n41) {defaultValue=true, parameterName=UseObjectPos}, B=CameraPositionWS())) {constB=0.05}
  n43 = Multiply(
   A=ScalarParameter() {defaultValue=13, group=NOOOTUse, parameterName=CRTNO_Outline-InnerSize1}
   B=Power(Base=Divide(A=@n39, B=@n42)) {constExponent=0.6}
  )
  n44 = Multiply(A=@n43, B=ScalarParameter() {defaultValue=1.5, group=NOOOTUse, parameterName=CRTNO_Outline-InnerSizeMobileMultiplier})
  n45 = Multiply(
   A=Clamp(None=Ceil(None=@n37)) {maxDefault=2, minDefault=1}
   B=Multiply(A=SceneTexelSize(), B=ShadingPathSwitch(Default=@n43, Forward=@n44, Mobile=@n44))
  ) {constA=2}
  n46 = Constant()
  n47 = AppendVector(A=ComponentMask(None=@n45) {desc=x, r=true}, B=@n46)
  n48 = ComponentMask(None=SceneTexture(UVs=Add(A=@n36.ViewportUV, B=@n47) {desc=x+}) {sceneTextureId=PPI_CustomDepth}.Color) {r=true}
  n49 = AppendVector(A=@n46, B=ComponentMask(None=@n45) {desc=y, g=true}) {desc=y+}
  n50 = ComponentMask(None=SceneTexture(UVs=Add(A=@n36.ViewportUV, B=@n49) {desc=y+}) {sceneTextureId=PPI_CustomDepth}.Color) {r=true}
  n51 = Constant() {r=-1}
  n52 = SceneTexture(UVs=Add(A=@n36.ViewportUV, B=Multiply(A=@n47, B=@n51)) {desc=x-}) {sceneTextureId=PPI_CustomDepth}
  n53 = Max(A=Max(A=Subtract(A=@n48, B=@n32), B=Subtract(A=@n50, B=@n32)), B=Subtract(A=ComponentMask(None=@n52.Color) {r=true}, B=@n32))
  n54 = Add(A=@n36.ViewportUV, B=Multiply(A=@n51, B=@n49)) {desc=y-}
  n55 = Subtract(A=ComponentMask(None=SceneTexture(UVs=@n54) {sceneTextureId=PPI_CustomDepth}.Color) {r=true}, B=@n32)
  n56 = ScalarParameter() {defaultValue=10, group=NOOOTUse, parameterName=DepthCompareT}
  n57 = ScalarParameter() {defaultValue=50, group=NOOOTUse, parameterName=DepthCompareT2}
  n58 = Fresnel(ExponentIn=Constant() {r=1}, BaseReflectFractionIn=Constant())
  n59 = OneMinus(None=Power(Base=OneMinus(None=@n58), Exp=ScalarParameter() {defaultValue=2, group=NOOOTUse, parameterName=DCF}))
  n60 = Constant()
  n61 = Multiply(
   A=If(A=@n32, B=Multiply(A=@n33, B=@n34.R), "A > B"=Constant(), "A == B"=@n35, "A < B"=@n35)
   B=If(A=Max(A=@n53, B=@n55), B=LinearInterpolate(A=@n56, B=@n57, Alpha=@n59), "A > B"=Constant() {r=1}, "A == B"=@n60, "A < B"=@n60)
  )
  n62 = TwoSidedSign()
  LinearInterpolate(
   A=ShadingPathSwitch(Default=Add(A=@n28, B=@n29.RGB), Forward=@n29.RGB, Mobile=@n29.RGB)
   B=ShadingPathSwitch(Default=LinearInterpolate(A=@n31, B=@n30.RGB, Alpha=@n61), Forward=@n27.RGB, Mobile=@n27.RGB)
   Alpha=Saturate(None=OneMinus(None=@n62))
  )
  MaterialFunctionCall(Emissive=Constant()) {materialFunction=&MF_FX_EmissiveAdjustments}
  n63 = VectorParameter() {parameterName=Center, defaultValue=<rgba 0.5 0.1 0 1>}
  n64 = AppendVector(A=@n63.R, B=@n63.R)
  n65 = Clamp(None=DynamicParameter() {paramNames [RightDot,AreaPower,Str,Appear]}.Appear) {minDefault=0.05}
  n66 = Divide(A=Add(A=TextureCoordinate(), B=Multiply(A=@n64) {constB=-1}), B=@n65)
  n67 = VectorParameter() {parameterName=PreScaleAndOffset, defaultValue=<rgba 0.9 0.9 0 0.0833>}
  n68 = Add(
   A=MaterialFunctionCall(UVs=Add(A=@n66, B=@n64), Texture Scale=AppendVector(A=@n67.R, B=@n67.R)) {materialFunction=&ScaleUVsByCenter}.UVs
   B=AppendVector(A=@n67.B, B=@n67.B)
  )
  n69 = MaterialFunctionCall(Float2=@n68) {materialFunction=&BreakOutFloat2Components}
  n70 = Multiply(A=@n69.R)
  n71 = Multiply(A=@n70) {constB=2}
  n72 = LinearInterpolate(A=@n70, B=OneMinus(None=@n70), Alpha=Ceil(None=Add(A=@n71) {constB=-1}))
  n73 = VectorParameter() {parameterName=CenterDistortParam, defaultValue=<rgba 4 -0.4 1.1 0>}
  n74 = Multiply(A=Power(Base=Multiply(A=@n72) {constB=2}, Exp=@n73.R) {constExponent=4}, B=Power(Base=@n69.G, Exp=@n73.B) {constExponent=3})
  n75 = VectorParameter() {parameterName=FX_LowestValues, channelNames {a=Distort2U, b=Distort1V, g=Distort1U, r=CenterDistortStrength}, defaultValue=<rgba -0.2 0 2 5>}
  n76 = LinearInterpolate(
   A=@n75.CenterDistortStrength
   B=@n73.G
   Alpha=DynamicParameter() {parameterIndex=1, paramNames [Time, ForwardDot, TempFIxWPOSpeed, TempFIxWPOStr]}.ForwardDot
  )
  n77 = VectorParameter() {parameterName=MasterShapeUV_TileAndOffset, defaultValue=<rgba 0.6 0.8 0 0.12>}
  n78 = Add(
   A=MaterialFunctionCall(X=Constant(), Y=Multiply(A=@n74, B=@n76) {constB=-0.4}) {materialFunction=&MakeFloat2}.Result
   B=MaterialFunctionCall(UVs=@n68, Texture Scale=AppendVector(A=@n77.R, B=@n77.R)) {materialFunction=&ScaleUVsByCenter}.UVs
  )
  n79 = AppendVector(A=@n77.B, B=@n77.B)
  n80 = VectorParameter() {parameterName=FX_LowestValues2, channelNames {a="", b="", g=Distort2MaskPower, r=Distort1MaskPower}, defaultValue=<rgba 1.5 2.5 0 0>}
  n81 = VectorParameter() {parameterName=PannerAndMaskPower_01, defaultValue=<rgba -0.0788 -0.7091 3 0>}
  n82 = DynamicParameter() {parameterIndex=1, paramNames [Time, ForwardDot, TempFIxWPOSpeed, TempFIxWPOStr]}
  n83 = LinearInterpolate(A=@n80.Distort1MaskPower, B=@n81.B, Alpha=@n82.ForwardDot)
  n84 = VectorParameter() {parameterName=TileAndOffset_01, defaultValue=<rgba 2 1.5 0 0>}
  n85 = Multiply(A=@n68, B=AppendVector(A=@n84.R, B=@n84.R))
  n86 = AppendVector(A=@n84.B, B=@n84.B)
  n87 = TextureSample(UVs=Panner(Coordinate=Add(A=@n85, B=@n86), Time=@n11.Time, Speed=AppendVector(A=@n81.R, B=@n81.R))) {samplerType=SAMPLERTYPE_Normal, texture=/Game/Global/Textures/Default/Noise/T_VIperGlitchNoise_N}
  n88 = VectorParameter() {parameterName=DistortOffsetAndStrength_01, defaultValue=<rgba 0 0 0 13>}
  n89 = AppendVector(A=@n75.Distort1U, B=@n75.Distort1U)
  n90 = AppendVector(A=@n88.B, B=@n88.B)
  n91 = Multiply(
   A=Add(A=OneMinus(None=@n69.G), B=ScalarParameter() {defaultValue=0.14, parameterName=StrengthMask1Offset}) {constB=0.14}
   B=ScalarParameter() {defaultValue=0.9, parameterName=StrengthMask1ScaleBeforeOffset}
  )
  n92 = ScalarParameter() {defaultValue=1, parameterName=Strength1Power}
  n93 = Multiply(
   A=Add(A=AppendVector(A=@n87.R, B=@n87.R), B=AppendVector(A=@n88.R, B=@n88.R))
   B=Multiply(A=LinearInterpolate(A=@n89, B=@n90, Alpha=@n10.ForwardDot), B=Power(Base=@n91, Exp=@n92))
  )
  n94 = ScalarParameter() {defaultValue=3, parameterName=NoiseClamp1}
  n95 = Multiply(A=@n94) {constB=-1}
  n96 = VectorParameter() {parameterName=PannerAndMaskPower_02, defaultValue=<rgba 0 -0.3758 3 0>}
  n97 = Power(Base=@n69.G, Exp=LinearInterpolate(A=@n80.Distort2MaskPower, B=@n96.B, Alpha=@n82.ForwardDot)) {constExponent=3}
  n98 = VectorParameter() {parameterName=TileAndOffset_02, defaultValue=<rgba 2 0.7 0 0>}
  n99 = Multiply(A=@n68, B=AppendVector(A=@n98.R, B=@n98.R))
  n100 = AppendVector(A=@n98.B, B=@n98.B)
  n101 = TextureSample(UVs=Panner(Coordinate=Add(A=@n99, B=@n100), Time=@n11.Time, Speed=AppendVector(A=@n96.R, B=@n96.R))) {
   samplerType=SAMPLERTYPE_Normal
   texture=/BRCosmetics/Gadgets/Assets/VinderTech_GliderChute/Glider_Teriyaki_Fish_Toon/FX/Texture/T_SwirlNoise_N
  }
  n102 = VectorParameter() {parameterName=DistortOffsetAndStrength_02, defaultValue=<rgba 0 0 10 15>}
  n103 = AppendVector(A=@n75.Distort2U, B=ScalarParameter() {defaultValue=10, parameterName=FX_Loest_Distort2V})
  n104 = AppendVector(A=@n102.B, B=@n102.B)
  n105 = Multiply(
   A=Add(A=OneMinus(None=@n69.G), B=ScalarParameter() {defaultValue=0.14, parameterName=StrengthMask2Offset}) {constB=0.14}
   B=ScalarParameter() {defaultValue=0.85, parameterName=StrengthMask2ScaleBeforeOffset}
  )
  n106 = ScalarParameter() {defaultValue=3, parameterName=Strength2Power}
  n107 = Multiply(
   A=Add(A=AppendVector(A=@n101.R, B=@n101.R), B=AppendVector(A=@n102.R, B=@n102.R))
   B=Multiply(A=LinearInterpolate(A=@n103, B=@n104, Alpha=@n82.ForwardDot), B=Power(Base=@n105, Exp=@n106))
  )
  n108 = ScalarParameter() {defaultValue=50000, parameterName=NoiseClamp2}
  n109 = Add(
   A=Multiply(A=Power(Base=@n69.G, Exp=@n83) {constExponent=3}, B=Clamp(None=@n93, Min=@n95, Max=@n94))
   B=Clamp(None=Multiply(A=@n97, B=@n107), Min=Multiply(A=@n108) {constB=-1}, Max=@n108)
  )
  n110 = TextureSampleParameter2D(UVs=Add(A=Add(A=@n78, B=@n79), B=Multiply(A=@n109, B=@n65))) {
   parameterName=Texture
   samplerSource=SSM_Clamp_WorldGroupSettings
   texture=/BRCosmetics/Gadgets/Assets/VinderTech_GliderChute/Glider_Teriyaki_Fish_Toon/FX/Texture/T_TeriyakiToon_WaterFlowBaseShape
  }
  Reroute(None=@n110.RGB)
  VectorParameter() {group=NOOOTUse, parameterName=BC, defaultValue=<rgba 0 0 0 1>}
 ]
}
