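# Translucent unlit Niagara ribbon material for the ink trail on the Teriyaki Fish Toon glider.
# A time-scrolled sine and panning swirl noise distort a shape mask and a scrolling dissolve texture, which multiply into a high-gain mask.
# Dynamic parameters drive dissolve, sine frequency, sine strength and noise strength per particle.
# The graph also holds a cartoon outline branch using four-tap custom-depth edge detection, scaled by distance and FOV.
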Material /BRCosmetics/Gadgets/Assets/VinderTech_GliderChute/Glider_Teriyaki_Fish_Toon/FX/Material/M_TeriyakiToon_InkTrail
graph {
 properties {
  allowTranslucentCustomDepthWrites=true
  bUsedWithNiagaraRibbons=true
  blendMode=BLEND_Translucent
  pixelDepthOffsetMode=PDOM_Legacy
  shadingModel=MSM_Unlit
  usedShadingModels=Unlit
 }
 nodes [
  Abs(None=Sine(None=Time()))
  n1 = ComponentMask(None=SceneTexture() {sceneTextureId=PPI_CustomDepth}.Color) {r=true}
  n2 = Multiply(A=ScalarParameter() {defaultValue=1.5, parameterName=CullDepth}, B=PixelDepth().R)
  n3 = Constant()
  n4 = Constant() {r=1}
  n5 = ScreenPosition()
  n6 = Subtract(A=Divide(A=ViewSize(), B=Constant() {r=320}), B=Constant() {r=2}) {constB=0}
  n7 = ViewSize()
  n8 = Divide(
   A=Divide(A=ComponentMask(None=@n7) {r=true}, B=ComponentMask(None=@n7) {g=true})
   B=MaterialFunctionCall() {materialFunction=/Engine/Functions/Engine_MaterialFunctions02/Utility/FOV}."FOV "
  )
  n9 = ObjectPositionWS()
  n10 = ParticlePositionWS()
  n11 = Multiply(A=Distance(A=StaticSwitchParameter(True=@n9, False=@n10) {defaultValue=true, parameterName=UseObjectPos}, B=CameraPositionWS())) {constB=0.05}
  n12 = Multiply(
   A=ScalarParameter() {defaultValue=7, group=Cartoon, parameterName=CRTNO_Outline-InnerSize1}
   B=Power(Base=Divide(A=@n8, B=@n11)) {constExponent=0.6}
  )
  n13 = Multiply(A=@n12, B=ScalarParameter() {defaultValue=1.5, group=Cartoon, parameterName=CRTNO_Outline-InnerSizeMobileMultiplier})
  n14 = Multiply(
   A=Clamp(None=Ceil(None=@n6)) {maxDefault=2, minDefault=1}
   B=Multiply(A=SceneTexelSize(), B=ShadingPathSwitch(Default=@n12, Forward=@n13, Mobile=@n13))
  ) {constA=2}
  n15 = Constant()
  n16 = AppendVector(A=ComponentMask(None=@n14) {desc=x, r=true}, B=@n15)
  n17 = Add(A=@n5.ViewportUV, B=@n16) {desc=x+}
  n18 = Subtract(A=ComponentMask(None=SceneTexture(UVs=@n17) {sceneTextureId=PPI_CustomDepth}.Color) {r=true}, B=@n1)
  n19 = AppendVector(A=@n15, B=ComponentMask(None=@n14) {desc=y, g=true}) {desc=y+}
  n20 = Add(A=@n5.ViewportUV, B=@n19) {desc=y+}
  n21 = Subtract(A=ComponentMask(None=SceneTexture(UVs=@n20) {sceneTextureId=PPI_CustomDepth}.Color) {r=true}, B=@n1)
  n22 = Constant() {r=-1}
  n23 = Multiply(A=@n16, B=@n22)
  n24 = ComponentMask(None=SceneTexture(UVs=Add(A=@n5.ViewportUV, B=@n23) {desc=x-}) {sceneTextureId=PPI_CustomDepth}.Color) {r=true}
  n25 = SceneTexture(UVs=Add(A=@n5.ViewportUV, B=Multiply(A=@n22, B=@n19)) {desc=y-}) {sceneTextureId=PPI_CustomDepth}
  n26 = Max(A=Max(A=Max(A=@n18, B=@n21), B=Subtract(A=@n24, B=@n1)), B=Subtract(A=ComponentMask(None=@n25.Color) {r=true}, B=@n1))
  n27 = OneMinus(None=Fresnel(ExponentIn=Constant() {r=1}, BaseReflectFractionIn=Constant()))
  n28 = ScalarParameter() {defaultValue=2, parameterName=DCF}
  n29 = LinearInterpolate(
   A=ScalarParameter() {defaultValue=5, parameterName=DepthCompareT}
   B=ScalarParameter() {defaultValue=20, parameterName=DepthCompareT2}
   Alpha=OneMinus(None=Power(Base=@n27, Exp=@n28))
  )
  n30 = Constant() {r=1}
  n31 = Constant()
  LinearInterpolate(
   A=VectorParameter() {parameterName=InnerColor, defaultValue=<rgba 0.6 0.6 0.6 1>}.RGB
   B=VectorParameter() {parameterName=OutColor, defaultValue=<rgba 0 0 0 1>}.RGB
   Alpha=Multiply(A=If(A=@n1, B=@n2, "A > B"=@n3, "A == B"=@n4, "A < B"=@n4), B=If(A=@n26, B=@n29, "A > B"=@n30, "A == B"=@n31, "A < B"=@n31))
  )
  MaterialFunctionCall(Emissive=Constant3Vector()) {materialFunction=&MF_FX_EmissiveAdjustments}
  Multiply()
  n32 = TextureCoordinate()
  n33 = MaterialFunctionCall(UVs=@n32, Texture Scale=Constant() {r=0.5}) {materialFunction=&ScaleUVsByCenter}
  n34 = MaterialFunctionCall(Float2=MaterialFunctionCall(
    X=MaterialFunctionCall(Float2=@n32) {materialFunction=&BreakOutFloat2Components}.R
    Y=MaterialFunctionCall(Float2=@n33.UVs) {materialFunction=&BreakOutFloat2Components}.G
   ) {materialFunction=&MakeFloat2}.Result) {materialFunction=&BreakOutFloat2Components}
  n35 = MaterialFunctionCall(X=@n34.G, Y=@n34.G) {materialFunction=&MakeFloat2}
  n36 = DynamicParameter() {paramNames [Dissolve,SineF,NoiseStr,SineStr]}
  n37 = Time() {bOverride_Period=true, period=6000}
  n38 = Add(
   A=Multiply(
    A=MaterialFunctionCall(Float2=@n35.Result) {materialFunction=&BreakOutFloat2Components}.G
    B=@n36.SineF
   )
   B=Multiply(A=@n37, B=ScalarParameter() {defaultValue=-1, parameterName=SineSPeed})
  )
  n39 = ScalarParameter() {defaultValue=0.2, parameterName=OffsetStr}
  n40 = DynamicParameter() {paramNames [Dissolve,SineF,NoiseStr,SineStr]}
  n41 = MaterialFunctionCall(X=Multiply(A=Sine(None=@n38), B=Multiply(A=@n39, B=@n40.SineStr)), Y=Constant()) {materialFunction=&MakeFloat2}
  n42 = Add(
   A=MaterialFunctionCall(UVs=@n35.Result, Texture Scale=ScalarParameter() {defaultValue=0.5, parameterName=2}) {materialFunction=&ScaleUVsByCenter}.UVs
   B=@n41.Result
  )
  n43 = VectorParameter() {parameterName=TileAndOffset_01, defaultValue=<rgba 0.5 0.5 0 0>}
  n44 = Multiply(A=@n35.Result, B=AppendVector(A=@n43.R, B=@n43.R))
  n45 = AppendVector(A=@n43.B, B=@n43.B)
  n46 = VectorParameter() {parameterName=PannerAndMaskPower_01, defaultValue=<rgba 0 -0.5 0 0>}
  n47 = TextureSample(UVs=Panner(Coordinate=Add(A=@n44, B=@n45), Time=Time() {bOverride_Period=true, period=6000}, Speed=AppendVector(A=@n46.R, B=@n46.R))) {
   samplerType=SAMPLERTYPE_Normal
   texture=/BRCosmetics/Gadgets/Assets/VinderTech_GliderChute/Glider_Teriyaki_Fish_Toon/FX/Texture/T_SwirlNoise_N
  }
  n48 = AppendVector(A=@n47.R, B=@n47.R)
  n49 = VectorParameter() {parameterName=DistortOffsetAndStrength_01, defaultValue=<rgba 0 0 0.05 0.05>}
  n50 = AppendVector(A=@n49.R, B=@n49.R)
  n51 = Multiply(A=Multiply(A=@n36.NoiseStr, B=Add(A=@n48, B=@n50)), B=AppendVector(A=@n49.B, B=@n49.B))
  n52 = Multiply(A=ScalarParameter() {defaultValue=15, parameterName=ShapeNoiseScael}, B=@n51)
  n53 = VectorParameter() {parameterName=TileAndOffset_02, defaultValue=<rgba 1 0.6485 0 -1>}
  n54 = Add(A=Multiply(A=@n35.Result, B=AppendVector(A=@n53.R, B=@n53.R)), B=Multiply(A=@n37, B=AppendVector(A=@n53.B, B=@n53.B)))
  n55 = ScalarParameter() {defaultValue=0.5, parameterName=DissolveDistortScale}
  n56 = Multiply(A=@n51, B=@n55)
  n57 = TextureSample(UVs=Add(A=Add(A=@n54, B=@n56), B=Multiply(A=@n55, B=@n41.Result))) {texture=/Game/CharClasses/Outlander/Gadgets/AntiMaterialCharge/Materials/T_AntiMaterialCharge_Dissolve_01}
  n58 = LinearInterpolate(A=OneMinus(None=@n57.R), B=@n57.R, Alpha=ScalarParameter() {defaultValue=1, parameterName=UseInvertDissolve})
  n59 = LinearInterpolate(
   A=ScalarParameter() {defaultValue=0.5, parameterName=D}
   B=ScalarParameter() {defaultValue=-0.9, parameterName=D2}
   Alpha=DynamicParameter() {paramNames [Dissolve,SineF,NoiseStr,SineStr]}.Dissolve
  )
  Multiply(
   A=MaterialFunctionCall!(TexCoord=Add(A=@n42, B=@n52), ClampChannel=VectorParameter() {defaultValue=<rgba 1 0 0 1>}.RGB) {materialFunction=/Game/Effects/Fort_Effects/MaterialFunctions/MF_FX_TilingMethod}.Result
   B=Multiply(A=Add(A=@n58, B=@n59), B=ScalarParameter() {defaultValue=100, parameterName=G})
  )
 ]
}
