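# Unlit, masked, two-sided Niagara ribbon and sprite material for the small ink flow effect on the Teriyaki Fish Toon glider.
# Dynamic parameters scale UVs built from V and add a sine wobble that grows along U.
# Those distorted UVs sample a scrolling dissolve noise texture and shape edge and V gradients.
# Subtracting the gradients from the noise, then applying Ceil, yields a hard-edged, toon-style shape.
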Material /BRCosmetics/Gadgets/Assets/VinderTech_GliderChute/Glider_Teriyaki_Fish_Toon/FX/Material/M_TeriyakiToon_SmallInkFlow
graph {
 properties {
  bUsedWithNiagaraRibbons=true
  bUsedWithNiagaraSprites=true
  blendMode=BLEND_Masked
  pixelDepthOffsetMode=PDOM_Legacy
  shadingModel=MSM_Unlit
  twoSided=true
  usedShadingModels=Unlit
 }
 nodes [
  n1 = MaterialFunctionCall(Float2=TextureCoordinate()) {materialFunction=&BreakOutFloat2Components}
  n2 = MaterialFunctionCall(X=@n1.G, Y=@n1.G) {materialFunction=&MakeFloat2}
  n3 = DynamicParameter() {defaultValue=<rgba 0.5 1 1 0>, paramNames [ScaleByCenter, SineFrequency, SineStrength, SineTimer]}
  n4 = MaterialFunctionCall(UVs=@n2.Result, Texture Scale=@n3.ScaleByCenter) {materialFunction=&ScaleUVsByCenter}
  n5 = MaterialFunctionCall(Float2=TextureCoordinate()) {materialFunction=&BreakOutFloat2Components}
  n6 = Multiply(A=@n5.R, B=@n3.SineFrequency)
  n7 = Multiply(A=@n3.SineTimer, B=ScalarParameter() {defaultValue=1}) {constB=-2}
  n8 = Multiply(A=Sine(None=Add(A=@n6, B=@n7)), B=@n3.SineStrength)
  n9 = Add(
   A=MaterialFunctionCall(
    X=MaterialFunctionCall(Float2=@n4.UVs) {materialFunction=&BreakOutFloat2Components}.R
    Y=MaterialFunctionCall(Float2=@n2.Result) {materialFunction=&BreakOutFloat2Components}.G
   ) {materialFunction=&MakeFloat2}.Result
   B=MaterialFunctionCall(X=Multiply(A=@n8, B=@n5.R), Y=Constant()) {materialFunction=&MakeFloat2}.Result
  )
  n10 = VectorParameter() {parameterName=Uv, defaultValue=<rgba 0.5 0.5 0.2 -1>}
  n11 = Add(A=Multiply(A=@n9, B=AppendVector(A=@n10.R, B=@n10.R)), B=Multiply(A=@n3.SineTimer, B=AppendVector(A=@n10.B, B=@n10.B)))
  n12 = MaterialFunctionCall(Float2=@n9) {materialFunction=&BreakOutFloat2Components}
  n13 = Saturate(None=Multiply(A=@n12.R) {constB=3})
  n14 = Saturate(None=Multiply(A=OneMinus(None=@n12.R)) {constB=3})
  n15 = Saturate(None=Max(A=OneMinus(None=@n13), B=OneMinus(None=@n14)))
  Ceil(None=Subtract(
    A=TextureSample(UVs=@n11) {texture=/Game/CharClasses/Outlander/Gadgets/AntiMaterialCharge/Materials/T_AntiMaterialCharge_Dissolve_01}.R
    B=Max(A=@n12.G, B=@n15)
   ) {constB=0.9})
  Time()
  Time()
 ]
}
