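# Masked, unlit, two-sided Niagara ribbon and sprite material for toon speed lines on the Teriyaki Fish glider, with emissive from particle color.
# An absolute sine along V, driven by dynamic scale and offset, is thresholded with Ceil against a thickness that varies along V, cutting hard-edged lines.
# These lines multiply a borrowed Outlander dissolve texture, weighted by inverted V, shifted by particle-driven Dissolve and boosted for sharp erosion.
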Material /BRCosmetics/Gadgets/Assets/VinderTech_GliderChute/Glider_Teriyaki_Fish_Toon/FX/Material/M_TeriyakiToon_SpeedLine
graph {
 properties {
  bUsedWithNiagaraRibbons=true
  bUsedWithNiagaraSprites=true
  blendMode=BLEND_Masked
  pixelDepthOffsetMode=PDOM_Legacy
  shadingModel=MSM_Unlit
  twoSided=true
  usedShadingModels=Unlit
 }
 nodes [
  MaterialFunctionCall(Emissive=ParticleColor().RGB) {materialFunction=&MF_FX_EmissiveAdjustments}
  n1 = TextureCoordinate()
  n2 = VectorParameter() {parameterName=DissolveUV, defaultValue=<rgba 1 2 0 0>}
  n3 = AppendVector(A=@n2.R, B=@n2.R)
  n4 = OneMinus(None=MaterialFunctionCall(Float2=TextureCoordinate()) {materialFunction=&BreakOutFloat2Components}.G)
  n5 = Multiply(
   A=TextureSample(UVs=Multiply(A=@n1, B=@n3)) {texture=/Game/CharClasses/Outlander/Gadgets/AntiMaterialCharge/Materials/T_AntiMaterialCharge_Dissolve_01}.R
   B=@n4
  )
  n6 = DynamicParameter() {defaultValue=<rgba 1.7 0 0.95 1>, paramNames [UVScale, UVOffset, LineThickness, Dissolve]}
  n7 = LinearInterpolate(A=ScalarParameter() {parameterName=D}, B=ScalarParameter() {defaultValue=-1, parameterName=D2}, Alpha=@n6.Dissolve)
  n8 = Multiply(
   A=MaterialFunctionCall(Float2=TextureCoordinate()) {materialFunction=&BreakOutFloat2Components}.G
   B=@n6.UVScale
  ) {constB=1.7}
  n9 = Abs(None=Sine(None=Add(A=@n8, B=@n6.UVOffset)))
  n10 = LinearInterpolate(
   A=@n6.LineThickness
   B=DynamicParameter() {parameterIndex=1, defaultValue=<rgba 0.85 0 0 0>, paramNames [LineThickness2, "", "", ""]}.LineThickness2
   Alpha=@n4
  )
  Multiply(A=Multiply(A=Add(A=@n5, B=@n7)) {constB=100}, B=Ceil(None=Subtract(A=@n9, B=@n10) {constB=0.95}))
 ]
}
