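# Translucent, unlit, two-sided toon water-splash sprite material for the Teriyaki Fish Toon glider's Niagara effects.
# Color is InnerColor with an outline, tinted by particle color; the IsMask switch picks a splash-texture threshold or custom-depth edge detection.
# The depth outline samples four neighbors, scaling width by distance, FOV and viewport.
# Splash alpha is cut by a scrolling, hard-edged noise dissolve with a vertical gradient, driven by a dynamic parameter.
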Material /BRCosmetics/Gadgets/Assets/VinderTech_GliderChute/Glider_Teriyaki_Fish_Toon/FX/Material/M_TeriyakiToon_Splash2
graph {
 properties {
  allowTranslucentCustomDepthWrites=true
  bUsedWithNiagaraSprites=true
  blendMode=BLEND_Translucent
  pixelDepthOffsetMode=PDOM_Legacy
  shadingModel=MSM_Unlit
  twoSided=true
  usedShadingModels=Unlit
 }
 nodes [
  n1 = Time() {bOverride_Period=true, period=600}
  n2 = DynamicParameter() {defaultValue=<rgba 0 1 1 1>}
  Add(A=@n1, B=@n2.Param2)
  n3 = ParticleColor()
  n4 = VectorParameter() {parameterName=InnerColor, defaultValue=<rgba 0.6 0.6 0.6 1>}
  n5 = Multiply(A=@n3.RGB, B=@n4.RGB)
  n6 = TextureSample() {texture=/Game/Effects/Fort_Effects/Textures/Water/T_water_splash_04}
  n7 = ScalarParameter() {defaultValue=0.07, parameterName=Outline}
  n8 = Saturate(None=Ceil(None=Subtract(A=@n6.B, B=@n7)))
  n9 = ComponentMask(None=SceneTexture() {sceneTextureId=PPI_CustomDepth}.Color) {r=true}
  n10 = Multiply(A=ScalarParameter() {defaultValue=1.5, parameterName=CullDepth}, B=PixelDepth().R)
  n11 = Constant()
  n12 = Constant() {r=1}
  n13 = ScreenPosition()
  n14 = Subtract(A=Divide(A=ViewSize(), B=Constant() {r=320}), B=Constant() {r=2}) {constB=0}
  n15 = ViewSize()
  n16 = Divide(
   A=Divide(A=ComponentMask(None=@n15) {r=true}, B=ComponentMask(None=@n15) {g=true})
   B=MaterialFunctionCall() {materialFunction=/Engine/Functions/Engine_MaterialFunctions02/Utility/FOV}."FOV "
  )
  n17 = ObjectPositionWS()
  n18 = ParticlePositionWS()
  n19 = Multiply(A=Distance(A=StaticSwitchParameter(True=@n17, False=@n18) {defaultValue=true, parameterName=UseObjectPos}, B=CameraPositionWS())) {constB=0.05}
  n20 = Multiply(
   A=ScalarParameter() {defaultValue=7, group=Cartoon, parameterName=CRTNO_Outline-InnerSize1}
   B=Power(Base=Divide(A=@n16, B=@n19)) {constExponent=0.6}
  )
  n21 = Multiply(A=@n20, B=ScalarParameter() {defaultValue=1.5, group=Cartoon, parameterName=CRTNO_Outline-InnerSizeMobileMultiplier})
  n22 = Multiply(
   A=Clamp(None=Ceil(None=@n14)) {maxDefault=2, minDefault=1}
   B=Multiply(A=SceneTexelSize(), B=ShadingPathSwitch(Default=@n20, Forward=@n21, Mobile=@n21))
  ) {constA=2}
  n23 = Constant()
  n24 = AppendVector(A=ComponentMask(None=@n22) {desc=x, r=true}, B=@n23)
  n25 = Add(A=@n13.ViewportUV, B=@n24) {desc=x+}
  n26 = Subtract(A=ComponentMask(None=SceneTexture(UVs=@n25) {sceneTextureId=PPI_CustomDepth}.Color) {r=true}, B=@n9)
  n27 = AppendVector(A=@n23, B=ComponentMask(None=@n22) {desc=y, g=true}) {desc=y+}
  n28 = Add(A=@n13.ViewportUV, B=@n27) {desc=y+}
  n29 = Subtract(A=ComponentMask(None=SceneTexture(UVs=@n28) {sceneTextureId=PPI_CustomDepth}.Color) {r=true}, B=@n9)
  n30 = Constant() {r=-1}
  n31 = Multiply(A=@n24, B=@n30)
  n32 = ComponentMask(None=SceneTexture(UVs=Add(A=@n13.ViewportUV, B=@n31) {desc=x-}) {sceneTextureId=PPI_CustomDepth}.Color) {r=true}
  n33 = SceneTexture(UVs=Add(A=@n13.ViewportUV, B=Multiply(A=@n30, B=@n27)) {desc=y-}) {sceneTextureId=PPI_CustomDepth}
  n34 = Max(A=Max(A=Max(A=@n26, B=@n29), B=Subtract(A=@n32, B=@n9)), B=Subtract(A=ComponentMask(None=@n33.Color) {r=true}, B=@n9))
  n35 = OneMinus(None=Fresnel(ExponentIn=Constant() {r=1}, BaseReflectFractionIn=Constant()))
  n36 = ScalarParameter() {defaultValue=2, parameterName=DCF}
  n37 = LinearInterpolate(
   A=ScalarParameter() {defaultValue=5, parameterName=DepthCompareT}
   B=ScalarParameter() {defaultValue=20, parameterName=DepthCompareT2}
   Alpha=OneMinus(None=Power(Base=@n35, Exp=@n36))
  )
  n38 = Constant() {r=1}
  n39 = Constant()
  n40 = LinearInterpolate(
   A=@n4.RGB
   B=VectorParameter() {parameterName=OutColor, defaultValue=<rgba 0 0 0 1>}.RGB
   Alpha=Multiply(A=If(A=@n9, B=@n10, "A > B"=@n11, "A == B"=@n12, "A < B"=@n12), B=If(A=@n34, B=@n37, "A > B"=@n38, "A == B"=@n39, "A < B"=@n39))
  )
  n41 = ShadingPathSwitch(Default=@n40, Forward=@n40, Mobile=@n40)
  MaterialFunctionCall(Emissive=StaticSwitchParameter(True=LinearInterpolate(B=@n5, Alpha=@n8), False=Multiply(A=@n3.RGB, B=@n41)) {parameterName=IsMask}) {materialFunction=&MF_FX_EmissiveAdjustments}
  n42 = TextureCoordinate()
  n43 = VectorParameter() {parameterName=DissolveUV, defaultValue=<rgba 0.6 0.6 0 0.2>}
  n44 = AppendVector(A=@n43.R, B=@n43.R)
  n45 = AppendVector(A=@n43.B, B=@n43.B)
  n46 = TextureSample(UVs=Add(A=Multiply(A=@n42, B=@n44), B=Multiply(A=@n45, B=@n1))) {texture=/Game/CharClasses/Outlander/Gadgets/AntiMaterialCharge/Materials/T_AntiMaterialCharge_Dissolve_01}
  n47 = MaterialFunctionCall(Float2=TextureCoordinate()) {materialFunction=&BreakOutFloat2Components}
  n48 = Add(
   A=Multiply(A=OneMinus(None=@n46.R), B=OneMinus(None=@n47.G))
   B=LinearInterpolate(A=ScalarParameter() {parameterName=D}, B=ScalarParameter() {defaultValue=-1, parameterName=D2}, Alpha=@n2.Param1)
  )
  Multiply(A=@n6.A, B=Saturate(None=Multiply(A=@n48) {constB=100}))
 ]
}
